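# UI material for the SevenSlide progress bar: a segmented vertical bar with decorative trim and alpha-composite blending.
# SDF boxes, a trapezoid and rotated cuts build the anti-aliased bar and decoration, with ping-ponged UVs forming segment lines.
# Progress and ProgressDelta position the fill and an angled end cap, with a separate delta color.
# Fill color eases toward red as Progress rises, while WarningBlinking forces red and sine-pulses bar opacity.
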Material /SevenSlide/UI/Art/M_UI_SevenSlideUI_ProgressBar
graph {
 properties {bAutomaticallySetUsageInEditor=false, blendMode=BLEND_AlphaComposite, materialDomain=MD_UI, usedShadingModels=Default Lit}
 nodes [
  n1 = NamedRerouteUsage()
  n2 = Constant() {r=0.5}
  n3 = NamedRerouteUsage()
  n4 = NamedRerouteUsage()
  n5 = Multiply(A=NamedRerouteDeclaration(None=Max(A=@n3, B=@n4)) {name=ProgressWCap, nodeColor=<rgba 0 0 0 1>}, B=@n1)
  n6 = LinearInterpolate(
   A=ScalarParameter() {defaultValue=1, parameterName=BarOpacity}
   B=MaterialFunctionCall(Speed=Constant() {r=2}) {materialFunction=&DebugTimeSine}.Result
   Alpha=ScalarParameter() {parameterName=WarningBlinking}
  )
  Max(
   A=LinearInterpolate(
    A=Multiply(A=@n1, B=@n2) {constB=0.5}
    B=LinearInterpolate(A=@n5, B=@n1, Alpha=@n2)
    Alpha=NamedRerouteDeclaration(None=@n6) {name=BarOpacity, nodeColor=<rgba 0 0 0 1>}
   )
   B=Min(A=NamedRerouteUsage()) {constB=0.5}
  )
  n7 = Constant4Vector() {constant=<rgba 0.016 0.0222 0.1946 1>}
  n8 = Saturate(None=Subtract(A=NamedRerouteUsage(), B=NamedRerouteUsage()))
  n9 = LinearInterpolate(
   A=VectorParameter() {parameterName=ProgressColor, defaultValue=<rgba 0.0234 0.9465 1 1>}.RGBA
   B=VectorParameter() {parameterName=DeltaProgressColor, defaultValue=<rgba 0.8879 0.9911 1 1>}.RGBA
   Alpha=@n8
  )
  n10 = LinearInterpolate(A=Constant4Vector() {constant=<rgba 0.8879 0.0296 0.0578 1>}.RGBA, B=Constant4Vector() {constant=<rgba 1 0.4678 0.4397 1>}.RGBA, Alpha=@n8)
  n11 = MaterialFunctionCall("Value (0-1)"=LinearInterpolate(Alpha=ScalarParameter() {defaultValue=0.5, parameterName=Progress}) {constB=1.2}) {materialFunction=&MF_UI_EaseCurves}
  n12 = LinearInterpolate(
   A=LinearInterpolate(B=Multiply(A=@n7.RGBA) {constB=0.2}, Alpha=NamedRerouteUsage())
   B=LinearInterpolate(A=LinearInterpolate(A=@n9, B=@n10, Alpha=@n11.EaseInCubic), B=@n10, Alpha=ScalarParameter() {parameterName=WarningBlinking})
   Alpha=Multiply(A=NamedRerouteUsage(), B=NamedRerouteUsage()) {constB=0}
  )
  n13 = NamedRerouteUsage()
  n14 = NamedRerouteUsage()
  n15 = Constant4Vector() {constant=<rgba 0.1095 0.3515 0.5149 1>}
  Multiply(A=Max(A=Multiply(A=@n12, B=@n13), B=Multiply(A=@n14, B=@n15.RGBA)), B=VertexColor().A)
  NamedRerouteDeclaration(None=MaterialFunctionCall() {materialFunction=&GetUserInterfaceUV}.Normalized UV) {name=UV, nodeColor=<rgba 0 0 0 1>}
  n16 = MaterialFunctionCall!(DistanceFieldA=NamedRerouteUsage(), DistanceFieldB=NamedRerouteUsage(), BlendAmount) {materialFunction=/FortUILibrary/MaterialFunctions/Shapes/Modifiers/MF_UI_SDF_Subtraction}
  NamedRerouteDeclaration(None=MaterialFunctionCall(DistanceField=@n16.Exact) {materialFunction=&MF_UI_AntiAliasing}.Inner) {name=Bar Shape, nodeColor=<rgba 0 0 0 1>}
  n17 = ComponentMask(None=NamedRerouteUsage()) {g=true}
  n18 = LinearInterpolate(Alpha=ScalarParameter() {defaultValue=0.5, parameterName=Progress}) {constA=0.13, constB=0.84}
  NamedRerouteDeclaration(None=Step(Y=OneMinus(None=@n17), X=NamedRerouteDeclaration(None=@n18) {name=Progress, nodeColor=<rgba 0 0 0 1>}) {constX=0.5}) {name=Progress Bar, nodeColor=<rgba 0 0 0 1>}
  n19 = NamedRerouteUsage()
  n20 = ScalarParameter() {defaultValue=0.5, parameterName=ProgressDelta}
  n21 = AppendVector(A=Constant(), B=Add(A=LinearInterpolate(Alpha=@n20) {constA=0.13, constB=0.84}) {constB=-0.475})
  n22 = ComponentMask(None=MaterialFunctionCall(UVs=Add(A=@n19, B=@n21) {constB=0}, "Rotation Angle (0-1)"=Constant() {r=0.055}) {materialFunction=&CustomRotator}.Rotated Values) {g=true}
  NamedRerouteDeclaration(None=MaterialFunctionCall(DistanceField=OneMinus(None=@n22), Threshold=Constant() {r=0.5}) {materialFunction=&MF_UI_AntiAliasing}.Inner) {name=End Cap, nodeColor=<rgba 0 0 0 1>}
  n23 = ComponentMask(None=Multiply(A=NamedRerouteUsage(), B=ScalarParameter() {defaultValue=20, parameterName=SegmentCount})) {g=true}
  NamedRerouteDeclaration(None=MaterialFunctionCall(
    DistanceField=MaterialFunctionCall!(Value=@n23, Scale, Height) {materialFunction=/FortUILibrary/MaterialFunctions/Basics/MF_UI_PingPong}.Result
    Threshold=Constant() {r=0.25}
   ) {materialFunction=&MF_UI_AntiAliasing}.Inner) {name=BackgroundLines, nodeColor=<rgba 0 0 0 1>}
  n24 = MaterialFunctionCall(UVs=Add(A=NamedRerouteUsage(), B=Constant2Vector() {g=-0.212}), "Rotation Angle (0-1)"=Constant() {r=-0.1875}) {materialFunction=&CustomRotator}
  NamedRerouteDeclaration(None=Multiply(A=ComponentMask(None=@n24.Rotated Values) {r=true}) {constB=2}) {name=Bar Cut SDF, nodeColor=<rgba 0 0 0 1>}
  n25 = NamedRerouteUsage()
  n26 = Constant2Vector() {g=0.1}
  NamedRerouteDeclaration(None=MaterialFunctionCall(UVs=Add(A=@n25, B=@n26), Size=Constant2Vector() {g=0.93, r=0.06}, CornerRoundness=Constant() {r=-1}) {materialFunction=&MF_UI_Box}.SDF) {name=Bar SDF, nodeColor=<rgba 0 0 0 1>}
  n27 = MaterialFunctionCall(
   DistanceFieldA=MaterialFunctionCall!(DistanceFieldA=NamedRerouteUsage(), DistanceFieldB=NamedRerouteUsage(), BlendAmount) {materialFunction=/FortUILibrary/MaterialFunctions/Shapes/Modifiers/MF_UI_SDF_Union}.Exact
   DistanceFieldB=NamedRerouteUsage()
  ) {materialFunction=&MF_UI_SDF_Subtraction}
  n28 = Add(A=@n16.Exact) {constB=-0.05}
  NamedRerouteDeclaration(None=MaterialFunctionCall(DistanceField=MaterialFunctionCall(DistanceFieldA=@n27.Exact, DistanceFieldB=@n28) {materialFunction=&MF_UI_SDF_Subtraction}.Exact) {materialFunction=&MF_UI_AntiAliasing}.Inner) {name=Decoration, nodeColor=<rgba 0 0 0 1>}
  n29 = Add(A=NamedRerouteUsage(), B=Constant2Vector() {g=-0.04, r=0.21})
  n30 = Constant() {r=-0.744}
  NamedRerouteDeclaration(None=MaterialFunctionCall!(
    UVs=MaterialFunctionCall(UVs=@n29, "Rotation Angle (0-1)"=@n30) {materialFunction=&CustomRotator}.Rotated Values
    TopWidth=Constant()
    BottomWidth=Constant() {r=0.9}
    Height=Constant() {r=0.6}
    OutlineThickness
    OutlineAlignment
    GlowMin
    GlowMax
   ) {materialFunction=/FortUILibrary/MaterialFunctions/Shapes/MF_UI_IsoscelesTrapezoid}.SDF) {name=Decor Triangle SDF, nodeColor=<rgba 0 0 0 1>}
  n31 = Add(A=NamedRerouteUsage(), B=Constant2Vector() {g=-0.35, r=0.05})
  n32 = Constant() {r=-0.03}
  NamedRerouteDeclaration(None=MaterialFunctionCall(
    UVs=MaterialFunctionCall(UVs=@n31, "Rotation Angle (0-1)"=@n32) {materialFunction=&CustomRotator}.Rotated Values
    Size=Constant2Vector() {g=0.1, r=0.25}
   ) {materialFunction=&MF_UI_Box}.SDF) {name=Decor Detail SDF, nodeColor=<rgba 0 0 0 1>}
  n33 = NamedRerouteUsage()
  n34 = Constant2Vector() {g=-0.02, r=0.5}
  NamedRerouteDeclaration(None=MaterialFunctionCall(UVs=Add(A=@n33, B=@n34), Size=Constant2Vector() {g=0.74, r=1}) {materialFunction=&MF_UI_Box}.SDF) {name=Decor Cut SDF, nodeColor=<rgba 0 0 0 1>}
 ]
}
