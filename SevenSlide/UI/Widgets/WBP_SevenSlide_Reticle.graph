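# HUD reticle widget that shows the SevenSlide item's fuel as a material-driven progress bar for the viewed pawn.
# On view-target change it listens for slider item events, then re-registers async gameplay-tag waits, or hides if fuel is not consumed.
# Powered-on controls visibility, activation starts a looping timer polling fuel attributes, overheat blinks the bar and EMP plays a disabled animation.
# Once fuel is full outside guided motion, polling stops and the bar fades after a curve-table delay.
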WidgetBlueprint /SevenSlide/UI/Widgets/WBP_SevenSlide_Reticle
animations [
 Anim_Intro {
  binding_count=1
  length_seconds=0.25
  object=&self:Anim_Intro
  sequence {bindings [Image_SlideProgress {
     id=#1
     possessed_class=Image
     tracks [{
       type=MovieSceneFloatTrack
       display_name=Render Opacity
       sections [{type=MovieSceneFloatSection, end_frame=5, start_frame=0, channels [None {type=MovieSceneScriptingFloatChannel, keys [[0,0],[4,1],[5,1]]}]}]
      }]
    }]}
 }
 Anim_Disabled {
  binding_count=1
  length_seconds=0.0834
  object=&self:Anim_Disabled
  sequence {
   bindings [Image_SlideProgress {
     id=#2
     possessed_class=Image
     tracks [
      {
       type=MovieSceneFloatTrack
       display_name=Render Opacity
       sections [{type=MovieSceneFloatSection, end_frame=5, start_frame=0, channels [None {type=MovieSceneScriptingFloatChannel, keys [[0,1],[5,0.25]]}]}]
      }
      {
       type=MovieSceneWidgetMaterialTrack
       sections [{
         type=MovieSceneParameterSection
         end_frame=5
         start_frame=0
         channels [
          ProgressColorR {type=MovieSceneScriptingFloatChannel, keys [[0,0.0234],[5,0.601]]}
          ProgressColorG {type=MovieSceneScriptingFloatChannel, keys [[0,0.9465],[5,0.988]]}
          ProgressColorB {type=MovieSceneScriptingFloatChannel, const [1,2]}
          ProgressColorA {type=MovieSceneScriptingFloatChannel, const [1,1]}
         ]
        }]
      }
     ]
    }]
   display_rate {numerator=60}
  }
 }
]
graphs [
 EventGraph {nodes [
   WaitSevenSliderOverheat() {is=title K2Node_CustomEvent}
   > n1 = Sequence()
   n2 = Get SevenSliderOverheat_Tag() -> (SevenSliderOverheat_Tag:Gameplay Tag Structure)
   n3 = Get PlayerPawn() -> (PlayerPawn:Fort Player Pawn Object Reference)
   > n4 = WaitGameplayTagAddToActor!(execute:Exec, TargetActor:Actor Object Reference=@n3, Tag:Gameplay Tag Structure=@n2, OnlyTriggerOnce:bool=false) -> (then:Exec, AsyncTaskProxy:Ability Async Wait Gameplay Tag Added Object Reference, Added:Exec) {is=title K2Node_AsyncAction}
   > Set Async_Overheat_Added(Async_Overheat_Added:Ability Async Wait Gameplay Tag Added Object Reference=@n4) -> (Output_Get:Ability Async Wait Gameplay Tag Added Object Reference)
   n5 = GetDynamicMaterial(execute=@n4.Added, self=Get Image_SlideProgress() -> (Image_SlideProgress:Image Object Reference))
   > SetScalarParameterValue(self=@n5, ParameterName=WarningBlinking, Value=1)
   n6 = WaitGameplayTagRemoveFromActor!(execute:Exec=@n1.then_1, TargetActor:Actor Object Reference=@n3, Tag:Gameplay Tag Structure=@n2, OnlyTriggerOnce:bool=false) -> (then:Exec, AsyncTaskProxy:Ability Async Wait Gameplay Tag Removed Object Reference, Removed:Exec) {is=title K2Node_AsyncAction}
   > Set Async_Overheat_Removed(Async_Overheat_Removed:Ability Async Wait Gameplay Tag Removed Object Reference=@n6) -> (Output_Get:Ability Async Wait Gameplay Tag Removed Object Reference)
   n7 = GetDynamicMaterial(execute=@n6.Removed, self=Get Image_SlideProgress())
   > SetScalarParameterValue(self=@n7, ParameterName=WarningBlinking, Value=0)
   WaitDisableStatesStarted() {is=title K2Node_CustomEvent}
   > n8 = Sequence()
   n9 = Get EMPTag() -> (EMPTag:Gameplay Tag Structure)
   n10 = Get PlayerPawn()
   > n11 = WaitGameplayTagAddToActor(TargetActor=@n10, Tag=@n9, OnlyTriggerOnce=false)
   > Set Async_Disabled_Added(Async_Disabled_Added:Ability Async Wait Gameplay Tag Added Object Reference=@n11) -> (Output_Get:Ability Async Wait Gameplay Tag Added Object Reference)
   n12 = Branch(execute=@n11.Added, Condition=Get IsFuelVisible() -> (IsFuelVisible:bool))
   > n13 = Clear and Invalidate Timer by Handle(Handle=Get FuelVisibility_TimerHandle() -> (FuelVisibility_TimerHandle:Timer Handle Structure))
   SetBarProgressAppearance!(execute=[@n13.then, @n14.then], self, Progress:double=0) -> (then)
   > PlayAnimationForward(InAnimation=Get Anim_Disabled() -> (Anim_Disabled:Widget Animation Object Reference))
   n14 = Set IsFuelVisible(execute=@n12.else, IsFuelVisible:bool=true) -> (Output_Get:bool)
   n15 = WaitGameplayTagRemoveFromActor(execute=@n8.then_1, TargetActor=@n10, Tag=@n9, OnlyTriggerOnce=false)
   > Set Async_Disabled_Removed(Async_Disabled_Removed:Ability Async Wait Gameplay Tag Removed Object Reference=@n15) -> (Output_Get:Ability Async Wait Gameplay Tag Removed Object Reference)
   Branch(execute=@n15.Removed, Condition=Get IsFuelVisible())
   > n16 = Branch(Condition=Is Timer Active by Handle!(Handle:Timer Handle Structure=Get UpdateProgress_TimerHandle() -> (UpdateProgress_TimerHandle:Timer Handle Structure)) -> (ReturnValue:bool))
   PlayAnimationReverse(execute=[@n16.then, @n18.then], InAnimation=Get Anim_Disabled())
   n17 = Set Timer by Event(execute=@n16.else, Delegate=Create Event(), Time=0.01, bLooping=true, bMaxOncePerFrame=true)
   > n18 = Set UpdateProgress_TimerHandle(UpdateProgress_TimerHandle:Timer Handle Structure=@n17) -> (Output_Get:Timer Handle Structure)
   n19 = OnViewTargetChanged!() -> (
    OutputDelegate
    then
    PlayerController:Fort Player Controller Object Reference
    OldViewTarget:Actor Object Reference
    NewViewTarget:Actor Object Reference
   ) {is=title K2Node_CustomEvent}
   > n20 = Cast To FortPlayerPawn(Object=@n19.NewViewTarget)
   > Set PlayerPawn(PlayerPawn:Fort Player Pawn Object Reference=@n20) -> (Output_Get:Fort Player Pawn Object Reference)
   > n21 = StartListeningToStatefulEvent!(
    execute:Exec
    Target:Gameplay Event Router Component Object Reference=GetGlobalEventRouter()
    EventType:Script Struct Object Reference=/SevenSlide/UI/Data/FortSevenSlider_SliderItemEvent
    Context:Object Reference
    EventBubblingRule:EEventBubblingRule Enum
   ) -> (
    then:Exec
    AsyncTaskProxy:Async Action Start Listening To Stateful Event Object Reference
    OnEventReceived:Exec
    OnSavedEventStateExists:Exec
    OnEventStateCleared:Exec
    Payload:Fort Seven Slider Slider Item Event Structure
   ) {is=title K2Node_StartListeningToStatefulEventAsyncAction}
   n22 = Break Fort Seven Slider Slider Item Event!(FortSevenSlider_SliderItemEvent:"Fort Seven Slider Slider Item Event Structure (by ref)"=@n21.Payload) -> (SevenSlider_6_#3:Fort Item Soft Object Reference) {is=title K2Node_BreakStruct}
   Branch(execute=[@n21.OnEventReceived, @n21.OnSavedEventStateExists], Condition=IsValidSoftObjectReference(SoftObjectReference=@n22))
   > n23 = Async Load Asset(Asset=@n22)
   n24 = Cast To FortWorldItem!(execute=@n23.Completed, Object=@n23) -> (then, CastFailed, AsFort World Item:Fort World Item Object Reference) {is=title K2Node_DynamicCast}
   > Set SevenSlideItem(SevenSlideItem:Fort World Item Object Reference=@n24) -> (Output_Get:Fort World Item Object Reference)
   > StopListeningAsyncActions()
   > n25 = HasMatchingGameplayTag!(execute:Exec, self:Object Reference=Get PlayerPawn(), TagToCheck:Gameplay Tag Structure=(TagName="sevenslide.DontConsumeFuel")) -> (then:Exec, ReturnValue:bool) {is=title K2Node_Message}
   > n26 = Branch(Condition=@n25)
   > SetVisibility(InVisibility=Hidden)
   WaitSevenSliderPoweredOn/OIff(execute=@n26.else)
   > WaitSevenSliderActive()
   > WaitSevenSliderOverheat() {is=title K2Node_CallFunction}
   > WaitDisableStatesStarted() {is=title K2Node_CallFunction}
   TickUpdateFuel() {is=title K2Node_CustomEvent}
   n27 = Get PlayerPawn()
   n28 = GetFloatAttribute!(
    Actor:Actor Object Reference=@n27
    Attribute:Gameplay Attribute Structure=Get FuelAttribute() -> (FuelAttribute:Gameplay Attribute Structure)
   ) -> (bSuccessfullyFoundAttribute:bool, ReturnValue:float)
   n29 = GetFloatAttribute(Actor=@n27, Attribute=Get MaxFuelAttribute() -> (MaxFuelAttribute:Gameplay Attribute Structure))
   > Branch(Condition=AND Boolean(A=@n28, B=@n29))
   > n30 = HasMatchingGameplayTag(self=Get PlayerPawn(), TagToCheck=(TagName="MovementMode.SevenSlide.GuidedMotion"))
   n31 = Remap Float!(Input=@n28.ReturnValue, OldRange_X:double=0, OldRange_Y:double=@n29.ReturnValue, NewRange=(X=0,Y=1), Clamp=true) -> (Output=~)
   > n32 = Branch(Condition=OR Boolean(A=@n30, B="float < float"(A=@n31, B=1)))
   > SetBarProgressAppearance(Progress=@n31)
   n33 = Clear and Invalidate Timer by Handle(execute=@n32.else, Handle=Get UpdateProgress_TimerHandle())
   n34 = HideFuelIndicator() {is=title K2Node_CustomEvent}
   n35 = Set Timer by Event(
    execute=@n33.then
    Delegate=@n34
    Time=GetValueAtLevel!(Input:"Scalable Float Structure (by ref)"=Get VisibilityDelay_SF() -> (VisibilityDelay_SF:Scalable Float Structure), Level:float=1) -> (ReturnValue:float)
    bLooping=false
   )
   > Set FuelVisibility_TimerHandle(FuelVisibility_TimerHandle:Timer Handle Structure=@n35) -> (Output_Get:Timer Handle Structure)
   WaitSevenSliderPoweredOn/OIff() {is=title K2Node_CustomEvent}
   > n36 = Sequence()
   n37 = Get PlayerPawn()
   n38 = Get SevenSliderPoweredOn_Tag() -> (SevenSliderPoweredOn_Tag:Gameplay Tag Structure)
   > n39 = WaitGameplayTagAddToActor(TargetActor=@n37, Tag=@n38, OnlyTriggerOnce=false)
   > Set Async_PoweredOn_Added(Async_PoweredOn_Added:Ability Async Wait Gameplay Tag Added Object Reference=@n39) -> (Output_Get:Ability Async Wait Gameplay Tag Added Object Reference)
   SetRenderOpacity(execute=@n39.Added, self=Get Image_SlideProgress(), InOpacity=0)
   > SetVisibility(InVisibility=SelfHitTestInvisible)
   > SetBarProgressAppearance(Progress=1)
   n40 = WaitGameplayTagRemoveFromActor(execute=@n36.then_1, TargetActor=@n37, Tag=@n38, OnlyTriggerOnce=false)
   > Set Async_PoweredOn_Removed(Async_PoweredOn_Removed:Ability Async Wait Gameplay Tag Removed Object Reference=@n40) -> (Output_Get:Ability Async Wait Gameplay Tag Removed Object Reference)
   HideFuelIndicator(execute=@n40.Removed) {is=title K2Node_CallFunction}
   > SetVisibility(InVisibility=Hidden)
   > Clear and Invalidate Timer by Handle(Handle=Get UpdateProgress_TimerHandle())
   > Clear and Invalidate Timer by Handle(Handle=Get FuelVisibility_TimerHandle())
   WaitSevenSliderActive() {is=title K2Node_CustomEvent}
   > n41 = WaitGameplayTagAddToActor(TargetActor=Get PlayerPawn(), Tag=Get SevenSliderActive_Tag() -> (SevenSliderActive_Tag:Gameplay Tag Structure), OnlyTriggerOnce=false)
   > Set Aysnc_Active_Added(Aysnc_Active_Added:Ability Async Wait Gameplay Tag Added Object Reference=@n41) -> (Output_Get:Ability Async Wait Gameplay Tag Added Object Reference)
   n42 = Branch(execute=@n41.Added, Condition=Get IsFuelVisible())
   > n43 = Clear and Invalidate Timer by Handle(Handle=Get FuelVisibility_TimerHandle())
   n44 = Set Timer by Event(execute=[@n43.then, @n45.then], Delegate=Create Event(), Time=0.01, bLooping=true, bMaxOncePerFrame=true)
   > Set UpdateProgress_TimerHandle(UpdateProgress_TimerHandle=@n44)
   PlayAnimationForward(execute=@n42.else, InAnimation=Get Anim_Intro() -> (Anim_Intro:Widget Animation Object Reference))
   > n45 = Set IsFuelVisible(IsFuelVisible=true)
   Event Construct()
   n46 = GetOwningPlayer()
   > n47 = Cast To Athena_PlayerController!(execute, Object=@n46) -> (then, CastFailed, AsAthena Player Controller:Athena Player Controller Object Reference) {is=title K2Node_DynamicCast}
   n48 = Create Event()
   > Bind Event to On View Target Changed(self:Fort Player Controller Object Reference=@n47, Delegate=@n48) {is=title K2Node_AddDelegate}
   > OnViewTargetChanged!(
    execute
    self
    PlayerController:Fort Player Controller Object Reference=@n47
    OldViewTarget:Actor Object Reference
    NewViewTarget:Actor Object Reference=GetViewTarget!(self:Controller Object Reference=@n47) -> (ReturnValue:Actor Object Reference)
   ) -> (then) {is=title K2Node_CallFunction}
   > SetVisibility(InVisibility=Hidden)
   Event Destruct()
   > n49 = Cast To Athena_PlayerController(Object=@n46)
   > Unbind Event from On View Target Changed(self:Fort Player Controller Object Reference=@n49, Delegate=@n48) {is=title K2Node_RemoveDelegate}
   > Clear and Invalidate Timer by Handle(Handle=Get UpdateProgress_TimerHandle())
   > Clear and Invalidate Timer by Handle(Handle=Get FuelVisibility_TimerHandle())
   > StopListeningAsyncActions()
   Set IsFuelVisible(execute=@n50.then, IsFuelVisible=false)
   n50 = PlayAnimationReverse(execute=@n34.then, InAnimation=Get Anim_Intro())
  ]}
 SetBarProgressAppearance {nodes [
   n1 = SetBarProgressAppearance!() -> (then, Progress:double=~) {is=title K2Node_FunctionEntry}
   > n2 = GetDynamicMaterial(self=Get Image_SlideProgress())
   n3 = float - float(A=1, B=@n1)
   > SetScalarParameterValue(self=@n2, ParameterName=Progress, Value=@n3)
   > SetScalarParameterValue(self=@n2, ParameterName=ProgressDelta, Value=@n3)
  ]}
 StopListeningAsyncActions {nodes [
   StopListeningAsyncActions() {is=title K2Node_FunctionEntry}
   > Cancel(self:Cancellable Async Action Object Reference=[
     Get Async_PoweredOn_Added() -> (Async_PoweredOn_Added:Ability Async Wait Gameplay Tag Added Object Reference)
     Get Async_PoweredOn_Removed() -> (Async_PoweredOn_Removed:Ability Async Wait Gameplay Tag Removed Object Reference)
    ])
   > Cancel(self=Get Aysnc_Active_Added() -> (Aysnc_Active_Added:Ability Async Wait Gameplay Tag Added Object Reference))
   > Cancel(self=[
     Get Async_Disabled_Added() -> (Async_Disabled_Added:Ability Async Wait Gameplay Tag Added Object Reference)
     Get Async_Disabled_Removed() -> (Async_Disabled_Removed:Ability Async Wait Gameplay Tag Removed Object Reference)
    ])
   > Cancel(self=[
     Get Async_Overheat_Added() -> (Async_Overheat_Added:Ability Async Wait Gameplay Tag Added Object Reference)
     Get Async_Overheat_Removed() -> (Async_Overheat_Removed:Ability Async Wait Gameplay Tag Removed Object Reference)
    ])
  ]}
]
variables [
 Async_Disabled_Added {type=AbilityAsync_WaitGameplayTagAdded, default=null}
 Async_Disabled_Removed {type=AbilityAsync_WaitGameplayTagRemoved, default=null}
 Async_Overheat_Added {type=AbilityAsync_WaitGameplayTagAdded, default=null}
 Async_Overheat_Removed {type=AbilityAsync_WaitGameplayTagRemoved, default=null}
 Async_PoweredOn_Added {type=AbilityAsync_WaitGameplayTagAdded, default=null}
 Async_PoweredOn_Removed {type=AbilityAsync_WaitGameplayTagRemoved, default=null}
 Aysnc_Active_Added {type=AbilityAsync_WaitGameplayTagAdded, default=null}
 EMPTag {type=GameplayTag, default {tagName=Gameplay.Status.EMP.Applied}}
 FuelAttribute {
  type=GameplayAttribute
  default {
   attribute=/Script/SevenSlideRuntime.SevenSlideAttributeSet:Fuel
   attributeName=Fuel
   attributeOwner=&SevenSlideAttributeSet
  }
 }
 FuelVisibility_TimerHandle {type=TimerHandle}
 IsFuelVisible {type=boolean}
 MaxFuelAttribute {
  type=GameplayAttribute
  default {
   attribute=&SevenSlideAttributeSet:MaxFuel
   attributeName=MaxFuel
   attributeOwner=&SevenSlideAttributeSet
  }
 }
 PlayerPawn {type=FortPlayerPawn}
 SevenSlideItem {type=FortWorldItem, default=null}
 SevenSliderActive_Tag {type=GameplayTag, default {tagName=sevenslide.activewindow}}
 SevenSliderOverheat_Tag {type=GameplayTag, default {tagName=sevenslide.Overheat}}
 SevenSliderPoweredOn_Tag {type=GameplayTag, default {tagName=sevenslide.PoweredOn}}
 UpdateProgress_TimerHandle {type=TimerHandle}
 VisibilityDelay_SF {
  type=ScalableFloat
  default {registryType=None, value=1, curve {curveTable=/SevenSlide/UI/Data/SevenSliderUIData, rowName=Default.SevenSlide.FuelDisplay.VisibilityDelay}}
 }
]
widgets {
 info {inheritedWidgetCount=1, parentClass=&FortHUDElementWidget, rootWidgetClass=&Overlay, widgetCount=3}
 widgets [
  Overlay_Root {widgetClassPath=&Overlay, widget=Overlay_Root}
  Image_SlideProgress {
   widgetClassPath=&Image
   bIsVariable=true
   parent=Overlay_Root
   widget=Image_SlideProgress
   properties {brush {
     bIsDynamicallyLoaded=false
     drawAs=Image
     imageType=FullColor
     mirroring=NoMirror
     resourceName=None
     resourceObject=/SevenSlide/UI/Art/M_UI_SevenSlideUI_ProgressBar
     tiling=NoTile
     imageSize=<120 120>
     margin=<ltrb 0 0 0 0>
     outlineSettings {
      bUseBrushTransparency=false
      roundingType=HalfHeightRadius
      width=0
      color {colorUseRule=UseColor_Specified, specifiedColor=<rgba 0 0 0 0>}
      cornerRadii=<0 0 0 0>
     }
     tintColor {colorUseRule=UseColor_Specified, specifiedColor=<rgba 1 1 1 1>}
     uVRegion {bIsValid=false, max=<0 0>, min=<0 0>}
    }}
   slot {class=OverlaySlot, properties {horizontalAlignment=HAlign_Right, verticalAlignment=VAlign_Center}}
  }
  TutorialHighlight {widgetClassPath=&TutorialHighlight}
 ]
}
